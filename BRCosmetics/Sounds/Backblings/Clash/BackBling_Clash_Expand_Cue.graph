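# Sound cue for the expand sound of the Clash back bling, a reactive cosmetic.
# An integer Stage parameter drives a switch that plays the small expand wave on the first three inputs, including the default, and the large wave on the last.
# A modulator at the root varies the result, which plays quietly through the backbling sound class with close-range pet attenuation.
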SoundCue /BRCosmetics/Sounds/Backblings/Clash/BackBling_Clash_Expand_Cue
root=Modulator_0
nodes [
 {id=Modulator_0, children [Switch_0]}
 {id=Switch_0, children [WavePlayer_2,WavePlayer_2,WavePlayer_2,WavePlayer_0], properties {intParameterName=Stage}}
 {id=WavePlayer_0, properties {soundWaveAssetPtr=&/Backpack_Clash_Expand_Large}}
 {id=WavePlayer_2, properties {soundWaveAssetPtr=&/Backpack_Clash_Expand_Small}}
]
properties {
 attenuationSettings=&Fort_Pets_Close_Athena_Att
 audioPropertiesSheet=/BRCosmetics/Sounds/Settings/PropertySheets/Reactive/APS_Cosmetic_Reactive_Backbling
 maxDistance=1800.0001
 soundClassObject=/Game/Sounds/FortSoundClassesAndModes/Fort_Backblings
 volumeMultiplier=0.2
}
